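# Niagara ribbon trail for the offhand CephaloChef pickaxe, built from skeletal-mesh anim trails.
# A single TrailTip emitter spawns two short-lived trails between bone/socket index pairs that share one index.
# Color is tinted by time of day, and alpha multiplies a fade curve by a time-of-day scalar and a system-age cutoff.
# An integer comparison on User.SwingSectionIndex picks between a faster and slower cutoff curve.
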NiagaraSystem /BRCosmetics/Weapons/FORT_Melee/Pickaxe_CephaloChef/FX/NS_Pickaxe_CephaloChefOffhand_Trail
dependencies {
 usedDataInterfaces [&NiagaraDataInterfaceCurve, &NiagaraDataInterfaceSkeletalMesh]
 usedDynamicInputs []
 usedModules [
  &self:SystemSpawnScript.SystemScriptSource.SystemScriptGraph.NiagaraNodeAssignment_2.SetVariables_#1
  &SystemState
 ]
}
emitters {TrailTip {
  emitterUpdateScript [
   &EmitterState()
   &FN_SpawnAnimTrail_v2(Number of Anim Trails=2)
  ]
  particleSpawnScript [
   /Game/Effects/Niagara/Modules/AnimTrail/FN_InitializeAnimTrail/FN_InitializeAnimTrailBoneIndecies(Second Bone or Socket Index=1)
   &FN_InitializeAnimTrailBoneIndecies(AnimTrail Selection=1, First Bone or Socket Index=2, Second Bone or Socket Index=1) {module=FN_InitializeAnimTrailBoneIndecies001}
   &FN_InitializeAnimTrail(Color=<rgba 0.4479 0.4811 0.5 1>, Lifetime=0.25, SkeletalMesh=$System.Skeletal Mesh) {module=FN_InitializeAnimTrail001}
  ]
  particleUpdateScript [
   &FN_UpdateAnimTrail(SkeletalMesh=$System.Skeletal Mesh) {module=FN_UpdateAnimTrail001}
   &ScaleColor(
    Scale Alpha=&Multiply_Float(
     A=&Multiply_Float(
      A=&FloatFromCurve(
       FloatCurve=DataInterface() {propertyValues {Curve {keys [
           {interpMode=RCIM_Cubic, leaveTangent=-2.3646, tangentMode=RCTM_Break, time=0, value=1}
           <at 0.2103 0.13>
           {arriveTangent=-1.0846, interpMode=RCIM_Cubic, leaveTangent=-1.0846, tangentMode=RCTM_User, time=1, value=0}
          ]}}}
       Scale Curve=0.75
      )
      B=&FNTimeOfDayScalar(Night Scale=0.5, Morning Scale=0.6, Evening Scale=0.5)
     )
     B=&MakeCustomFloatFromBool(
      Bool=&SetBoolByIntComparison(A=4, B=$User.SwingSectionIndex, Comparison Type=::NewEnumerator4)
      True Float=&FloatFromCurve(
       FloatCurve=DataInterface() {propertyValues {Curve {keys [
           {arriveTangent=0.6568, interpMode=RCIM_Cubic, leaveTangent=0.6568, tangentMode=RCTM_User, time=0.1, value=1}
           {arriveTangent=-104.3341, interpMode=RCIM_Cubic, leaveTangent=-104.3341, tangentMode=RCTM_User, time=0.1061, value=0.7737}
           {interpMode=RCIM_Cubic, leaveTangent=-14.1396, time=0.13, value=0}
          ]}}}
       CurveIndex=$Engine.System.Age
      )
      False Float=&FloatFromCurve(
       FloatCurve=DataInterface() {propertyValues {Curve {keys [
           {arriveTangent=0.6568, interpMode=RCIM_Cubic, leaveTangent=0.6568, tangentMode=RCTM_User, time=0.1, value=1}
           {arriveTangent=-19.1962, interpMode=RCIM_Cubic, leaveTangent=-19.1962, tangentMode=RCTM_User, time=0.1795, value=0.6024}
           {interpMode=RCIM_Cubic, leaveTangent=-14.1396, time=0.25, value=0}
          ]}}}
       CurveIndex=$Engine.System.Age
      )
     )
    )
    Scale RGB=&MakeVectorFromLinearColorRGB(Color=&FNTimeOfDayColor(
      Night Color=<rgba 0.4463 0.62 1 1>
      Morning Color=<rgba 1 0.6899 0.5573 1>
      Day Color=<rgba 1 0.7878 0.6667 1>
      Evening Color=<rgba 1 0.6291 0.4464 1>
     ))
    ScaleA=true
    ScaleRGB=true
   ) {module=ScaleColor002}
  ]
  rendererClasses [&NiagaraRibbonRendererProperties]
  renderers [{rendererClass=&NiagaraRibbonRendererProperties}]
 }}
summary {
 systemName=NS_Pickaxe_CephaloChefOffhand_Trail
 emitters [{emitterName=TrailTip, rendererClasses [&NiagaraRibbonRendererProperties]}]
 userVariables [
  User.bIsFrontEnd {type=NiagaraBool}
  User.bIsFrontEndPreview {type=NiagaraBool}
  User.RibbonIDOffset {type=NiagaraInt32}
  User.SwingSectionIndex {type=NiagaraInt32}
 ]
}
