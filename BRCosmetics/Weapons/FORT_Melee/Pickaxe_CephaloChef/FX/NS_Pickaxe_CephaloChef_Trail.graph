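# Niagara system for the CephaloChef pickaxe trail, drawn by a single emitter with two ribbon renderers.
# FN anim trail modules spawn, initialize and update short-lived particles from the system's skeletal mesh.
# Time-of-day values tint the trail and scale its alpha, while curves over particle life and system age fade it out.
# Normalized age overrides ribbon U0, and a similarly faded WaveColor attribute is set alongside dynamic material parameters.
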NiagaraSystem /BRCosmetics/Weapons/FORT_Melee/Pickaxe_CephaloChef/FX/NS_Pickaxe_CephaloChef_Trail
dependencies {
 usedDataInterfaces [&NiagaraDataInterfaceCurve, &NiagaraDataInterfaceColorCurve, &NiagaraDataInterfaceSkeletalMesh]
 usedDynamicInputs []
 usedModules [
  &self:SystemSpawnScript.SystemScriptSource.SystemScriptGraph.NiagaraNodeAssignment_0.SetVariables_#1
  &SystemState
 ]
}
emitters {Main_Trail {
  emitterUpdateScript [
   &EmitterState()
   &FN_SpawnAnimTrail(Spawn Interval=0)
  ]
  particleSpawnScript [
   &FN_InitializeAnimTrailBoneIndecies()
   &FN_InitializeAnimTrail(Color=<rgba 0.3299 0.3745 0.5 0.35>, Lifetime=0.135, SkeletalMesh=$System.Skeletal Mesh)
  ]
  particleUpdateScript [
   &FN_UpdateAnimTrail(Locked Particle RibbonWidth=1, SkeletalMesh=$System.Skeletal Mesh) {module=FN_UpdateAnimTrail001}
   &ScaleColor(
    Scale Alpha=&FloatFromCurve(
     FloatCurve=DataInterface() {propertyValues {Curve {keys [
         {arriveTangent=-40.3311, interpMode=RCIM_Cubic, leaveTangent=-40.3311, tangentMode=RCTM_User, time=0, value=2.5}
         {arriveTangent=-1.3327, interpMode=RCIM_Cubic, leaveTangent=-1.3327, tangentMode=RCTM_User, time=0.1, value=1.25}
         {interpMode=RCIM_Cubic, tangentMode=RCTM_User, time=1, value=0}
        ]}}}
     Scale Curve=&FN_Select_Float_from_Time_Of_Day(Morning=0.5, Day=0.4, Evening=0.5, Night=0.45)
    )
    Scale RGB=&MakeVectorFromLinearColorRGB(Color=&FNTimeOfDayColor(
      Night Color=<rgba 0.2 0.275 0.5 1>
      Morning Color=<rgba 0.5 0.45 0.4542 1>
      Day Color=<rgba 0.7 0.7 0.7 1>
      Evening Color=<rgba 0.65 0.4333 0.39 1>
     ))
    ScaleA=true
    ScaleRGB=true
   )
   &ScaleColor(
    Scale Alpha=&Multiply_Float(
     A=&FloatFromCurve(FloatCurve=DataInterface() {propertyValues {Curve {keys [
          {arriveTangent=-21.3684, interpMode=RCIM_Cubic, leaveTangent=-21.3684, tangentMode=RCTM_User, time=0.1, value=1}
          {arriveTangent=-6.7025, interpMode=RCIM_Cubic, leaveTangent=-6.7025, tangentMode=RCTM_User, time=0.1531, value=0.3765}
          <at 0.3 0 RCIM_Cubic>
         ]}}})
     B=&FloatFromCurve(
      FloatCurve=DataInterface() {propertyValues {Curve {keys [
          {arriveTangent=-1.669, interpMode=RCIM_Cubic, leaveTangent=-1.669, tangentMode=RCTM_User, time=0.1, value=1}
          {arriveTangent=-6.7025, interpMode=RCIM_Cubic, leaveTangent=-6.7025, tangentMode=RCTM_User, time=0.1972, value=0.4329}
          <at 0.35 0 RCIM_Cubic>
         ]}}}
      CurveIndex=$Engine.System.Age
     )
    )
    ScaleA=true
    ScaleRGB=true
   ) {module=ScaleColor001}
   /SetVariables(Particles.RibbonU0Override=$Particles.NormalizedAge) {isSetParameters=true}
   /SetVariables(Particles.WaveColor:LinearColor=&MultiplyLinearColor(
     Linear Color=<rgba 0.25 0.25 0.25 1>
     Scale Factor=&MakeVector4FromVectorAndFloat(
      "Vector (RGB)"=<1 1 1>
      "Float (Alpha)"=&Multiply_Float(
       A=&FloatFromCurve(FloatCurve=DataInterface() {propertyValues {Curve {keys [
            {arriveTangent=-21.3684, interpMode=RCIM_Cubic, leaveTangent=-21.3684, tangentMode=RCTM_User, time=0.1, value=1.5}
            {arriveTangent=-6.7025, interpMode=RCIM_Cubic, leaveTangent=-6.7025, tangentMode=RCTM_User, time=0.1531, value=0.3765}
            <at 0.3 0 RCIM_Cubic>
           ]}}})
       B=&FloatFromCurve(
        FloatCurve=DataInterface() {propertyValues {Curve {keys [
            {arriveTangent=-1.669, interpMode=RCIM_Cubic, leaveTangent=-1.669, tangentMode=RCTM_User, time=0.1, value=1}
            {arriveTangent=-6.7025, interpMode=RCIM_Cubic, leaveTangent=-6.7025, tangentMode=RCTM_User, time=0.1972, value=0.4329}
            <at 0.35 0 RCIM_Cubic>
           ]}}}
        CurveIndex=$Engine.System.Age
       )
      )
     )
    )) {isSetParameters=true}
   &DynamicMaterialParameters(
    Index 1 Param 1=1
    Module.Index1Float0Write=true
    Module.Index1Float1Write=false
    Module.Index1Float2Write=false
    Module.Index1Float3Write=false
    Write Parameter Index 0=::NewEnumerator0
    Write Parameter Index 1=::NewEnumerator1
   )
  ]
  rendererClasses [&NiagaraRibbonRendererProperties]
  renderers [
   {rendererClass=&NiagaraRibbonRendererProperties}
   {rendererClass=&NiagaraRibbonRendererProperties, rendererIndex=1}
  ]
 }}
summary {
 systemName=NS_Pickaxe_CephaloChef_Trail
 emitters [{emitterName=Main_Trail, rendererClasses [&NiagaraRibbonRendererProperties]}]
 userVariables [
  User.RibbonIDOffset {type=NiagaraInt32}
  User.SwingSectionIndex {type=NiagaraInt32}
 ]
}
